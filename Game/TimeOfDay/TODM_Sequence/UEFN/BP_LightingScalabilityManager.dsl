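; Lighting scalability manager that toggles listed actors and post process volumes by the player's Global Illumination quality level.
; Construction script and BeginPlay, on cosmetic-capable instances only, bind to quality changes and run ToggleActorVisibility.
; It matches sg.GlobalIlluminationQuality against per-actor flags, enabling volumes or setting root visibility, with optional child propagation blocked by tag.
; OnActorDestroyed removes the quality binding, though setup calls Unbind on OnDestroyed rather than Bind.
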(event EventBeginPlay
  (Networking|CanExecuteCosmeticEvents
    (:True
      (Utilities|FlowControl|Sequence
        (:then_0
          (EventDispatchers|BindEventtoOnGlobalIlluminationQualityChanged (Settings|GetFortGameUserSettings) (EventDispatchers|CreateEvent :!event Custom|RefreshScalability))
          (CallFunction|ToggleActorVisibility))
        (:then_1
          (Game|UnbindEventfromOnDestroyed (EventDispatchers|CreateEvent :!event Custom|OnActorDestroyed)))))))

(event Custom|RefreshScalability
  (CallFunction|ToggleActorVisibility))

(event Custom|OnActorDestroyed
  (params (DestroyedActor :type Actor%20Object%20Reference))
  (EventDispatchers|UnbindEventfromOnGlobalIlluminationQualityChanged (Settings|GetFortGameUserSettings) (EventDispatchers|CreateEvent :!event Custom|RefreshScalability)))

(fn ToggleActorVisibility
  (bind _propagateactorsvisibilitytochildren (Variables|Default|GetPropagateActorsVisibilityToChildren))
  (Networking|CanExecuteCosmeticEvents
    (:True
      (bind _returnvalue (Development|GetConsoleVariableIntValue "sg.GlobalIlluminationQuality"))
      (for _array_element (Variables|Default|GetActors)
        ;[0 1 0.6593] GYepes: Adding support for post process volumes
        (bind _aspost_process_volume (Utilities|Casting|CastToPostProcessVolume (Utilities|Struct|BreakLightingScalabilityStruct _array_element))
          (:then
            (Variables|/Script/Engine.PostProcessVolume|SetbEnabled (Math|Boolean|ORBoolean (and (== _returnvalue 0) (.Low_8_73FFDD6D43ABCE8782D8A4B9607DC8D3 (Utilities|Struct|BreakLightingScalabilityStruct _array_element))) (and (== _returnvalue 1) (.Medium_10_E4CF263249C007CE2682FE8B28B2C9D8 (Utilities|Struct|BreakLightingScalabilityStruct _array_element))) (and (== _returnvalue 2) (.High_12_184E8BD1439B911C4D1817BB46CBADB1 (Utilities|Struct|BreakLightingScalabilityStruct _array_element))) (and (== _returnvalue 3) (.Epic_14_B44452DB4ADE6F835ADF69A6F2A64536 (Utilities|Struct|BreakLightingScalabilityStruct _array_element))) (and (== _returnvalue 4) (.Cinematic_16_5958D8F840A5F310D2C461B67FD095BE (Utilities|Struct|BreakLightingScalabilityStruct _array_element)))) _aspost_process_volume))
          (:CastFailed
            (Rendering|SetVisibility (Variables|/Script/Engine.Actor|GetRootComponent (Utilities|Struct|BreakLightingScalabilityStruct _array_element)) (Math|Boolean|ORBoolean (and (== _returnvalue 0) (.Low_8_73FFDD6D43ABCE8782D8A4B9607DC8D3 (Utilities|Struct|BreakLightingScalabilityStruct _array_element))) (and (== _returnvalue 1) (.Medium_10_E4CF263249C007CE2682FE8B28B2C9D8 (Utilities|Struct|BreakLightingScalabilityStruct _array_element))) (and (== _returnvalue 2) (.High_12_184E8BD1439B911C4D1817BB46CBADB1 (Utilities|Struct|BreakLightingScalabilityStruct _array_element))) (and (== _returnvalue 3) (.Epic_14_B44452DB4ADE6F835ADF69A6F2A64536 (Utilities|Struct|BreakLightingScalabilityStruct _array_element))) (and (== _returnvalue 4) (.Cinematic_16_5958D8F840A5F310D2C461B67FD095BE (Utilities|Struct|BreakLightingScalabilityStruct _array_element)))) (and (not (Components|ComponentHasTag (Variables|/Script/Engine.Actor|GetRootComponent (Utilities|Struct|BreakLightingScalabilityStruct _array_element)) "DontPropagateVisibilityToChildren")) _propagateactorsvisibilitytochildren))
            ;[1 0.2173 0] GYepes: Disconnecting in order to resolve FORT-693225
            (Development|SetHiddeninGame (Variables|/Script/Engine.Actor|GetRootComponent (Utilities|Struct|BreakLightingScalabilityStruct _array_element)) (not (Math|Boolean|ORBoolean (and (== _returnvalue 0) (.Low_8_73FFDD6D43ABCE8782D8A4B9607DC8D3 (Utilities|Struct|BreakLightingScalabilityStruct _array_element))) (and (== _returnvalue 1) (.Medium_10_E4CF263249C007CE2682FE8B28B2C9D8 (Utilities|Struct|BreakLightingScalabilityStruct _array_element))) (and (== _returnvalue 2) (.High_12_184E8BD1439B911C4D1817BB46CBADB1 (Utilities|Struct|BreakLightingScalabilityStruct _array_element))) (and (== _returnvalue 3) (.Epic_14_B44452DB4ADE6F835ADF69A6F2A64536 (Utilities|Struct|BreakLightingScalabilityStruct _array_element))) (and (== _returnvalue 4) (.Cinematic_16_5958D8F840A5F310D2C461B67FD095BE (Utilities|Struct|BreakLightingScalabilityStruct _array_element))))) (and (not (Components|ComponentHasTag (Variables|/Script/Engine.Actor|GetRootComponent (Utilities|Struct|BreakLightingScalabilityStruct _array_element)) "DontPropagateVisibilityToChildren")) _propagateactorsvisibilitytochildren))))))))

(fn ConstructionScript
  (Networking|CanExecuteCosmeticEvents
    (:True
      (Utilities|FlowControl|Sequence
        (:then_0
          (EventDispatchers|BindEventtoOnGlobalIlluminationQualityChanged (Settings|GetFortGameUserSettings) (EventDispatchers|CreateEvent :!event RefreshScalability))
          (CallFunction|ToggleActorVisibility))
        (:then_1
          (Game|UnbindEventfromOnDestroyed (EventDispatchers|CreateEvent :!event OnActorDestroyed)))))))
