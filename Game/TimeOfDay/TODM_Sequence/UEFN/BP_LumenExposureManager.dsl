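; UEFN exposure manager actor that switches between Lumen and non-Lumen post-process components based on platform and settings.
; Lumen requires allowed GI, no Switch, and non-mobile or cloud; otherwise UseNonLumenPost gates the fallback.
; Construction, BeginPlay and GI quality changes refresh both components' enabled state and visibility.
; Construction, BeginPlay and a custom event also write a directional light's negated forward vector to the Fortnite global material parameter collection.
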(event Collision|EventActorBeginOverlap
  (params (OtherActor :type Actor%20Object%20Reference)))

(event EventTick
  (params (DeltaSeconds :type Float%20%28single-precision%29)))

(event Custom|Update%20Directional%20Light%20MPC
  (Default|SetMPC))

(event EventBeginPlay
  (CallFunction|UpdateComponents)
  (Default|SetMPC))

(event Scalability|EventOnGlobalIlluminationQualityChanged
  (CallFunction|UpdateComponents))

(fn IsLumenSupported
  (bind _returnvalue (TouchPlatform|IsMobileGame))
  (bind _returnvalue_2 (FortGlobals|IsRunninginCloud))
  (bind _returnvalue_3 (not _returnvalue))
  (bind _returnvalue_4 (TouchPlatform|IsNintendoSwitch))
  (bind _returnvalue_5 (not _returnvalue_4))
  (bind _returnvalue_6 (or _returnvalue_3 _returnvalue_2))
  (bind _returnvalue_7 (Settings|GetGameUserSettings))
  (Fortnite|FortShippingLog self (Utilities|String|Append "UEFN Exposure Device: Is Lumen Supported = " (Utilities|String|ToString(Boolean) (Math|Boolean|ANDBoolean (Settings|IsGlobalIlluminationAllowed _returnvalue_7) _returnvalue_5 _returnvalue_6))))
  (return (Math|Boolean|ANDBoolean (Settings|IsGlobalIlluminationAllowed _returnvalue_7) _returnvalue_5 _returnvalue_6)))

(fn SetMPC
  (bind _returnvalue (Actor|GetActorOfClass "/Script/Engine.DirectionalLight"))
  (Utilities|IsValid _returnvalue
    (:Is%20Valid
      (Rendering|Material|SetVectorParameterValue "/Game/Packages/Fortress_SharedMaterials/GlobalMaterialParameters/FortniteMaterialParameters.FortniteMaterialParameters" "SunlightDirectionalVector" (Math|Conversions|ToLinearColor(Vector) (* (Transformation|GetForwardVector (Variables|/Script/Engine.Actor|GetRootComponent _returnvalue)) -1.0)) :!class /Script/Engine.KismetMaterialLibrary))))

(fn UpdateComponents
  (bind _lumenpostprocess (Variables|Default|GetLumenPostProcess))
  (bind _nonlumenpostprocess (Variables|Default|GetNonLumenPostProcess))
  (bind _islumensupported (CallFunction|IsLumenSupported))
  (Utilities|FlowControl|Sequence
    (:then_0
      (bind _output_get (Variables|/Script/Engine.PostProcessComponent|SetbEnabled _islumensupported _lumenpostprocess))
      (Rendering|SetVisibility _lumenpostprocess _output_get false))
    (:then_1
      (bind _output_get_2 (Variables|/Script/Engine.PostProcessComponent|SetbEnabled (and (not _islumensupported) (Variables|Default|GetUseNonLumenPost)) _nonlumenpostprocess))
      (Rendering|SetVisibility _nonlumenpostprocess _output_get_2 false))))

(fn ConstructionScript
  (CallFunction|UpdateComponents)
  (Default|SetMPC))
